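; BP_Athena_RuneSeal is an island rune seal Blueprint for the Athena environment, showing a selectable rune symbol.
; Its construction script sets the relative rotation of the S_DFHeight component from RuneRotation on one axis.
; It then assigns material slot 0 on that component from the RuneMaterials array at RuneIndex, so each placed instance can show a different rune.
; BeginPlay has no logic.
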(event EventBeginPlay)

(fn ConstructionScript
  (Transformation|SetRelativeRotation (Variables|Default|GetS_DFHeight) (Math|Rotator|MakeRotator 0.0 0.0 (Variables|Default|GetRuneRotation)) false false)
  (Rendering|Material|SetMaterial (Variables|Default|GetS_DFHeight) 0 (Utilities|Array|Get(acopy) (Variables|Default|GetRuneMaterials) (Variables|Default|GetRuneIndex))))
