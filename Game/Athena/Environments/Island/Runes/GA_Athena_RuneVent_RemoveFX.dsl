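; Gameplay ability that manages the looping rune vent FX cue on an Athena player pawn and cleans it up afterward.
; On activation it adds the cue if the pawn is not in a vehicle, then watches for skydiving or a change to walking or custom movement, the latter covering water landings.
; These events, or a failsafe delay, remove the cue and the granting effect.
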(event Ability|EventActivateAbility
  (bind _asfort_player_pawn_athena (Utilities|Casting|CastToFortPlayerPawnAthena (Ability|GetAvatarActorfromActorInfo))
    (:then
      (Variables|Default|SetPlayerPawn _asfort_player_pawn_athena)
      (if (not (Vehicle|IsinVehicle _asfort_player_pawn_athena))
        (Ability|AddGameplayCueToOwner (Variables|Default|GetLooping%20GC))
        (EventDispatchers|BindEventtoOnBeginSkydiving (Utilities|Casting|CastToFortPlayerControllerAthena (Fortnite|GetFortPlayerControllerfromActor (Variables|Default|GetPlayerPawn))) (delegate Custom|RemoveOnSkydive))
        (Utilities|FlowControl|Sequence
          (:then_0
            (Ability|Tasks|WaitMovementModeChange "MOVE_Walking"
              (:then
                (Utilities|FlowControl|RetriggerableDelay 15.0) ; Failsafe to ensure ability ends.
              )
              (:OnChange
                _)))
          (:then_1
            (Ability|Tasks|WaitMovementModeChange "MOVE_Custom"
              (:OnChange
                _)) ; Fix for fx not getting removed when landing in water (FORT-223198)
          ))
        (Ability|RemoveGameplayCueFromOwner (Variables|Default|GetLooping%20GC))
        (Ability|RemoveGrantedbyEffect)))
    (:CastFailed
      (Ability|EndAbility))))

(event Custom|RemoveOnSkydive
  (Ability|RemoveGameplayCueFromOwner (Variables|Default|GetLooping%20GC))
  (Ability|RemoveGrantedbyEffect))
