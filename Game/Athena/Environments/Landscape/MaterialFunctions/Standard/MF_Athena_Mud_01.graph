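# Material function for the Athena mud landscape layer, producing material attributes plus separate tinted color, untinted color and height outputs.
# It samples the Mud texture parameter with a shared world-group wrap sampler, using world XY projection unless Coords is supplied.
# RGB is multiplied by the MudTint landscape parameter collection value, which an input can override.
# Texture alpha serves as height and is packed into Opacity, with fully rough shading.
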MaterialFunction /Game/Athena/Environments/Landscape/MaterialFunctions/Standard/MF_Athena_Mud_01
graph {nodes [
  n1 = ComponentMask(None=WorldPosition().XYZ) {g=true, r=true}
  n2 = Constant() {r=512}
  n3 = TextureSampleParameter2D(UVs=FunctionInput(Preview=Divide(A=@n1, B=@n2)) {bUsePreviewValueAsDefault=true, inputName=Coords, inputType=FunctionInput_Vector2}) {
   group=Mud
   parameterName=Mud
   samplerSource=SSM_Wrap_WorldGroupSettings
   texture=/Game/Athena/Environments/Landscape/Textures/Arid/T_Athena_Arid_MudVariant_01
  }
  n4 = CollectionParameter() {collection=/Game/Athena/Environments/Landscape/MPC/MPC_Landscape, parameterName=MudTint}
  n5 = Multiply(
   A=@n3.RGB
   B=FunctionInput(Preview=ComponentMask(None=@n4) {b=true, g=true, r=true}) {bUsePreviewValueAsDefault=true, inputName=Color Tint Override, sortPriority=1}
  )
  FunctionOutput(None=MakeMaterialAttributes!(
    BaseColor=FunctionOutput(None=@n5) {outputName=BaseColor}
    Metallic
    Specular
    Roughness=Constant() {r=1}
    Anisotropy
    EmissiveColor
    Opacity=FunctionOutput(None=@n3.A) {outputName=Height}
    OpacityMask
    Normal
    Tangent
    WorldPositionOffset
    SubsurfaceColor
    ClearCoat
    ClearCoatRoughness
    AmbientOcclusion
    Refraction
    CustomizedUVs_0
    CustomizedUVs_1
    CustomizedUVs_2
    CustomizedUVs_3
    CustomizedUVs_4
    CustomizedUVs_5
    CustomizedUVs_6
    CustomizedUVs_7
    PixelDepthOffset
    ShadingModel
    Displacement
   ))
  FunctionOutput(None=@n3.RGB) {outputName=BaseColor_Plain}
 ]}
