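# Lit surface material for a corrupted island vent, blending the Athena landscape mud into the vent's own textures.
# The mud layer desaturates and tints the shared mud function's color and takes specular from a world-XY projected gravel texture, sharing the vent normal.
# Vertex red drives the attribute blend; UseEmissive adds two-color glow and adds the ceiled emissive mask to that alpha.
# The result passes through MF_CreativeMoveTool.
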Material /Game/Athena/Environments/Island/Runes/Material/M_CorruptedVent_A
graph {
 properties {bUseMaterialAttributes=true, pixelDepthOffsetMode=PDOM_Legacy, usedShadingModels=Default Lit}
 nodes [
  n1 = MaterialFunctionCall!(Material=MaterialFunctionCall!(Coords, Color Tint Override) {materialFunction=/Game/Athena/Environments/Landscape/MaterialFunctions/Standard/MF_Athena_Mud_01}.Result) {materialFunction=/Engine/Functions/MaterialLayerFunctions/MatLayerBlend_BreakBaseColor}
  n2 = ScalarParameter() {defaultValue=0.75, group=Corruption, parameterName=Corruption_MudDesatPreTint}
  n3 = Multiply(
   A=ScalarParameter() {defaultValue=1}
   B=Multiply(A=VectorParameter() {parameterName=Gravel_Tint, defaultValue=<rgba 1.45 1.45 1.45 0>}.RGB, B=Desaturation(None=@n1.BaseColor, Fraction=@n2))
  )
  n4 = WorldPosition()
  n5 = TextureSampleParameter2D(UVs=Divide(A=ComponentMask(None=@n4.XYZ) {g=true, r=true}, B=ScalarParameter() {defaultValue=475, group=Corruption, parameterName=Corruption_TexScale})) {
   group=Gravel
   parameterName=Gravel
   samplerSource=SSM_Wrap_WorldGroupSettings
   texture=/Game/Athena/Environments/Landscape/Textures/Arid/T_Athena_Arid_MudVariant_01
  }
  n6 = LinearInterpolate(
   A=ScalarParameter() {defaultValue=-0.45, group=Gravel, parameterName=Gravel_SpecMin}
   B=ScalarParameter() {defaultValue=2, group=Gravel, parameterName=Gravel_SpecMax}
   Alpha=@n5.R
  )
  n7 = Constant() {r=1}
  n8 = TextureSampleParameter2D() {parameterName=Normal, samplerType=SAMPLERTYPE_Normal, texture=/Game/Athena/Environments/Island/Runes/Texture/T_CorruptedVent_A_N}
  n9 = TextureSampleParameter2D() {parameterName=Diffuse, texture=/Game/Athena/Environments/Island/Runes/Texture/T_CorruptedVent_A_D}
  n10 = TextureSampleParameter2D() {parameterName=Mask, samplerType=SAMPLERTYPE_Masks, texture=/Game/Athena/Environments/Island/Runes/Texture/T_CorruptedVent_A_S}
  n11 = Multiply(
   A=VectorParameter() {parameterName=EmissiveColor, defaultValue=<rgba 1 0.0103 1 1>}.RGB
   B=ScalarParameter() {defaultValue=8, parameterName=EmissivePower}
  )
  n12 = Multiply(
   A=VectorParameter() {parameterName=EmissiveColor_1, defaultValue=<rgba 0.5518 0.1107 0.6979 0>}.RGB
   B=ScalarParameter() {defaultValue=5, parameterName=EmissivePower_1}
  )
  n13 = TextureSample() {texture=/Game/Athena/Environments/Island/Runes/Texture/T_CorruptedVent_A_E}
  n14 = StaticSwitchParameter(True=LinearInterpolate(B=LinearInterpolate(A=@n11, B=@n12, Alpha=@n13.R) {constB=0}, Alpha=@n13.R), False=Constant()) {parameterName=UseEmissive}
  n15 = VertexColor()
  n16 = Ceil(None=@n13.R)
  n17 = Add(A=@n15.R, B=@n16)
  MaterialFunctionCall(MA In=BlendMaterialAttributes(
    A=MakeMaterialAttributes!(
     BaseColor=@n3
     Metallic
     Specular=@n6
     Roughness=@n7
     Anisotropy
     EmissiveColor
     Opacity
     OpacityMask
     Normal=@n8.RGB
     Tangent
     WorldPositionOffset
     SubsurfaceColor
     ClearCoat
     ClearCoatRoughness
     AmbientOcclusion
     Refraction
     CustomizedUVs_0
     CustomizedUVs_1
     CustomizedUVs_2
     CustomizedUVs_3
     CustomizedUVs_4
     CustomizedUVs_5
     CustomizedUVs_6
     CustomizedUVs_7
     PixelDepthOffset
     ShadingModel
     Displacement
    )
    B=MakeMaterialAttributes(BaseColor=@n9.RGB, Specular=@n10.R, Roughness=@n10.R, EmissiveColor=@n14, Normal=@n8.RGB)
    Alpha=StaticSwitchParameter(True=@n17, False=@n15.R) {parameterName=UseEmissive}
   )) {materialFunction=&MF_CreativeMoveTool}
  Subtract(A=@n15.R, B=@n16)
 ]
}
