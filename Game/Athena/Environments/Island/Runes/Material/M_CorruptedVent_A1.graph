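# Corrupted vent material for island runes, layering volcano geyser textures with an optional masked two-color emissive and switchable terrain ground.
# Static switches pick gravel pools, mud, lava, grass or plain road gravel as the ground, with height lerps shaping pools and mud.
# Vertex red blends vent into ground, plus the ceiled emissive mask when UseEmissive is on.
# Normals are world space, with flat up-facing ground, and the result passes through MF_CreativeMoveTool.
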Material /Game/Athena/Environments/Island/Runes/Material/M_CorruptedVent_A1
graph {
 properties {bTangentSpaceNormal=false, bUseMaterialAttributes=true, pixelDepthOffsetMode=PDOM_Legacy, usedShadingModels=Default Lit}
 nodes [
  n1 = WorldPosition()
  MaterialFunctionCall!(
   Coords=Divide(A=ComponentMask(None=@n1.XYZ) {g=true, r=true}, B=ScalarParameter() {defaultValue=475, group=Corruption, parameterName=Corruption_TexScale})
   Color Tint
  ) {materialFunction=/Game/Athena/Environments/Landscape/MaterialFunctions/Standard/MF_Athena_Gravel_01}
  n2 = TextureSampleParameter2D() {parameterName=Mask, samplerType=SAMPLERTYPE_Masks, texture=/Game/Athena/Environments/Volcano/Textures/T_Volcano_Geyser_S}
  n3 = VectorParameter() {parameterName=EmissiveColor, defaultValue=<rgba 1 0.0103 1 1>}
  n4 = ScalarParameter() {defaultValue=8, parameterName=EmissivePower}
  n5 = VectorParameter() {parameterName=EmissiveColor_1, defaultValue=<rgba 0.5518 0.1107 0.6979 0>}
  n6 = ScalarParameter() {defaultValue=5, parameterName=EmissivePower_1}
  n7 = TextureSample() {texture=/Game/Athena/Environments/Island/Runes/Texture/T_CorruptedVent_A_E}
  n8 = LinearInterpolate(B=LinearInterpolate(A=Multiply(A=@n3.RGB, B=@n4), B=Multiply(A=@n5.RGB, B=@n6), Alpha=@n7.R) {constB=0}, Alpha=@n7.R)
  n9 = Constant()
  n10 = GetMaterialAttributes(None=MakeMaterialAttributes!(
    BaseColor=TextureSampleParameter2D() {parameterName=Diffuse, texture=/Game/Athena/Environments/Volcano/Textures/T_Volcano_Geyser_D}.RGB
    Metallic
    Specular=@n2.R
    Roughness=@n2.R
    Anisotropy
    EmissiveColor=StaticSwitchParameter(True=@n8, False=@n9) {parameterName=UseEmissive}
    Opacity
    OpacityMask
    Normal=TextureSampleParameter2D() {parameterName=Normal, samplerType=SAMPLERTYPE_Normal, texture=/Game/Athena/Environments/Volcano/Textures/T_Volcano_Geyser_N}.RGB
    Tangent
    WorldPositionOffset
    SubsurfaceColor
    ClearCoat
    ClearCoatRoughness
    AmbientOcclusion
    Refraction
    CustomizedUVs_0
    CustomizedUVs_1
    CustomizedUVs_2
    CustomizedUVs_3
    CustomizedUVs_4
    CustomizedUVs_5
    CustomizedUVs_6
    CustomizedUVs_7
    PixelDepthOffset
    ShadingModel
    Displacement
   )) {attributeGetTypes [#1]}
  n11 = Transform(None=@n10.Normal)
  n12 = MaterialFunctionCall() {materialFunction=/Game/Athena/Environments/Landscape/MaterialFunctions/Standard/MF_Athena_RoadGravel}
  n13 = Add(
   A=Multiply(A=Constant() {r=1}, B=ScalarParameter() {defaultValue=1.5, parameterName=GravelPool-X})
   B=ScalarParameter() {defaultValue=-1, parameterName=GravelPool-+}
  )
  n14 = LinearInterpolate(
   A=VectorParameter() {parameterName=GravelPoolMud, defaultValue=<rgba 0.12 0.0641 0.036 0>}.RGB
   B=Multiply(A=@n12.Gravel, B=Constant3Vector() {constant=<rgba 1.5 1.1733 0.8777 0>})
   Alpha=MaterialFunctionCall(Transition Phase=Clamp(None=@n13), Height Texture=@n12.Gravel_Height) {materialFunction=&HeightLerp}.Lerp Alpha No Contrast
  )
  n15 = Constant3Vector() {constant=<rgba 0.9844 0.7659 0.6562 0>}
  n16 = SetMaterialAttributes!(
   MaterialAttributes
   Base Color=ShadingPathSwitch(Default=@n14, Deferred=@n14, Mobile=Multiply(A=@n12.Gravel, B=@n15)) {desc=Migrated from FeatureLevelSwitch}
   Specular=Constant() {r=0.2}
   Roughness=Constant() {r=1}
   Normal=Constant3Vector() {constant=<rgba 0 0 1 0>}
  ) {attributeSetTypes [
    #2
    #3
    #4
    #1
   ]}
  n17 = MaterialFunctionCall!(UVs) {materialFunction=/Game/Athena/Environments/Landscape/MaterialFunctions/Standard/MF_Athena_Forest_01}
  n18 = Constant3Vector() {constant=<rgba 1.5 1.2944 0.7891 1>}
  n19 = Multiply(
   A=LinearInterpolate(A=Constant() {r=-0.25}, B=Constant() {r=1}, Alpha=@n12.Gravel_NoColor)
   B=Constant3Vector() {constant=<rgba 0.1354 0.0916 0.0585 0>}
  )
  n20 = Constant3Vector() {constant=<rgba 3 1.6777 1.132 1>}
  n21 = Constant() {r=0.5}
  n22 = OneMinus(None=@n17.Height)
  n23 = SetMaterialAttributes(
   Base Color=LinearInterpolate(
    A=Multiply(A=@n17.BaseColor, B=@n18)
    B=Multiply(A=@n19, B=@n20)
    Alpha=MaterialFunctionCall(Transition Phase=@n21, Height Texture=@n22) {materialFunction=&HeightLerp}.Lerp Alpha No Contrast
   )
   Specular=Constant() {r=0.2}
   Roughness=Constant() {r=1}
   Normal=Constant3Vector() {constant=<rgba 0 0 1 0>}
  ) {attributeSetTypes [
    #2
    #3
    #4
    #1
   ]}
  n24 = GetMaterialAttributes(None=MaterialFunctionCall() {materialFunction=/Game/Athena/Environments/Landscape/MaterialFunctions/Standard/MF_Athena_Grass_01}.Result) {attributeGetTypes [#2]}
  n25 = MaterialFunctionCall!(BaseColor=@n24.Base Color) {materialFunction=/Game/Athena/Environments/Landscape/MaterialFunctions/Standard/MF_Athena_GrassMatch}
  n26 = SetMaterialAttributes(Base Color=@n12.Gravel, Specular=Constant() {r=0.2}, Roughness=Constant() {r=0.9}) {attributeSetTypes [
    #2
    #3
    #4
    #1
   ]}
  n27 = StaticSwitchParameter(
   True=MaterialFunctionCall!(UV, WorldAlignedSize=Constant() {r=921}) {materialFunction=/Game/Environments/Sets/Vulcano/Materials/MF_LavaGround}.Result
   False=StaticSwitchParameter(
    True=SetMaterialAttributes!(MaterialAttributes=@n24.MaterialAttributes, Base Color=@n25.Result) {attributeSetTypes [#2]}
    False=@n26
   ) {parameterName=Use_Grass, sortPriority=2}
  ) {parameterName=Use_Lava, sortPriority=2}
  n28 = GetMaterialAttributes(None=StaticSwitchParameter(True=@n16, False=StaticSwitchParameter(True=@n23, False=@n27) {parameterName=Use_Mud, sortPriority=3}) {parameterName=Use_Pools, sortPriority=4}) {attributeGetTypes [#1]}
  n29 = Constant3Vector() {constant=<rgba 0 0 1 0>}
  n30 = VertexColor()
  n31 = Ceil(None=@n7.R)
  n32 = Add(A=@n30.R, B=@n31)
  MaterialFunctionCall(MA In=BlendMaterialAttributes(
    A=SetMaterialAttributes!(MaterialAttributes=@n10.MaterialAttributes, Normal=@n11) {attributeSetTypes [#1]}
    B=SetMaterialAttributes(MaterialAttributes=@n28.MaterialAttributes, Normal=@n29) {attributeSetTypes [#1]}
    Alpha=StaticSwitchParameter(True=@n32, False=@n30.R) {parameterName=UseEmissive}
   )) {materialFunction=&MF_CreativeMoveTool}
  Subtract(A=@n30.R, B=@n31)
 ]
}
